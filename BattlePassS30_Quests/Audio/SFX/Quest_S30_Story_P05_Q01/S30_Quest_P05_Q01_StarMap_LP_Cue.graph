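# Sound cue for the Season 30 story quest star map, playing a layered reveal sound in the quests sound class.
# A mixer combines two layers at equal volume, one picking randomly from four reveal waves and the other from three "Lyr" variants.
# Each layer has its own modulator for slight pitch and volume variation.
# It borrows a hidden-star ambient loop attenuation asset, and though named LP, the graph has no looping node.
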SoundCue /BattlePassS30_Quests/Audio/SFX/Quest_S30_Story_P05_Q01/S30_Quest_P05_Q01_StarMap_LP_Cue
root=Mixer_0
nodes [
 {id=Mixer_0, children [Modulator_0,Modulator_1], properties {inputVolume [1,1]}}
 {id=Modulator_0, children [Random_0], properties {pitchMax=1.02, pitchMin=0.98, volumeMax=1.02, volumeMin=0.98}}
 {id=Modulator_1, children [Random_1], properties {pitchMax=1.02, pitchMin=0.98, volumeMax=1.02, volumeMin=0.98}}
 {id=Random_0, children [WavePlayer_0,WavePlayer_2,WavePlayer_3,WavePlayer_4], properties {weights [1,1,1,1]}}
 {id=Random_1, children [WavePlayer_5,WavePlayer_7,WavePlayer_8], properties {weights [1,1,1]}}
 {id=WavePlayer_0, properties {soundWaveAssetPtr=&/S30_Quest_P05_Star_Map_Reveal_01}}
 {id=WavePlayer_2, properties {soundWaveAssetPtr=&/S30_Quest_P05_Star_Map_Reveal_02}}
 {id=WavePlayer_3, properties {soundWaveAssetPtr=&/S30_Quest_P05_Star_Map_Reveal_03}}
 {id=WavePlayer_4, properties {soundWaveAssetPtr=&/S30_Quest_P05_Star_Map_Reveal_04}}
 {id=WavePlayer_5, properties {soundWaveAssetPtr=&/S30_Quest_P05_Star_Map_Reveal_Lyr_04}}
 {id=WavePlayer_7, properties {soundWaveAssetPtr=&/S30_Quest_P05_Star_Map_Reveal_Lyr_02}}
 {id=WavePlayer_8, properties {soundWaveAssetPtr=&/S30_Quest_P05_Star_Map_Reveal_Lyr_03}}
]
properties {
 attenuationSettings=/Game/Athena/Sounds/SupplyDrops/Fort_HiddenStar_Ambient_Loop_Att
 maxDistance=2500
 soundClassObject=&Fort_Quests
 volumeMultiplier=1
}
